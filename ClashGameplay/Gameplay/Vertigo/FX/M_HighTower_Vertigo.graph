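# Masked, two-sided lit material for the HighTower Vertigo effect, usable on Niagara mesh particles, that stretches a mesh toward a TargetPosition.
# World position offset combines that stretch with a curve-atlas cone profile, normal-texture displacement and a sine/cosine wobble, behind a static switch.
# Panned or sphere-mask UVs drive a three-color swirl blend and noise dissolve.
# Emissive mixes sky-light-direction highlights with a tinted texture, multiplied by time-of-day color.
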Material /ClashGameplay/Gameplay/Vertigo/FX/M_HighTower_Vertigo
graph {
 properties {bUsedWithNiagaraMeshParticles=true, blendMode=BLEND_Masked, pixelDepthOffsetMode=PDOM_Legacy, twoSided=true, usedShadingModels=Default Lit}
 nodes [
  ActorPositionWS()
  n1 = Add(
   A=DynamicParameter() {defaultValue=<rgba 0 1 1 1>, paramNames [Dissolve,Param2,Param3,Param4]}.Dissolve
   B=ScalarParameter() {parameterName=Dissolve}
  )
  n2 = ScalarParameter() {defaultValue=0.5, parameterName=Mask_Add}
  n3 = Time()
  n4 = ComponentMask(None=TextureCoordinate()) {r=true}
  n5 = ScalarParameter() {defaultValue=1, parameterName=U_Scale}
  n6 = SphereMask(A=WorldPosition().XYZ, B=ActorPositionWS(), Radius=ScalarParameter() {defaultValue=3000, parameterName=Radius}) {hardnessPercent=0}
  n7 = ScalarParameter() {defaultValue=1, parameterName=V_Scale}
  n8 = TextureCoordinate()
  n9 = VectorParameter() {parameterName=TextureCtrl, defaultValue=<rgba 1 1 0 0>}
  n10 = AppendVector(A=@n9.R, B=@n9.R)
  n11 = LinearInterpolate(
   A=AppendVector(A=Multiply(A=@n4, B=@n5), B=Multiply(A=@n6, B=@n7))
   B=Panner(Coordinate=Multiply(A=@n8, B=@n10), Speed=AppendVector(A=@n9.B, B=@n9.B))
   Alpha=ScalarParameter() {parameterName=UV_Style}
  )
  n12 = MaterialFunctionCall(In=TextureSample(UVs=Add(A=@n11, B=@n3)) {texture=/Game/Global/Textures/Default/Noise/T_FireNoise_01}.R, Contrast=Constant()) {materialFunction=&CheapContrast}
  Add(A=LinearInterpolate(A=@n1, B=Add(A=@n1, B=@n2) {constB=0.5}, Alpha=Multiply(A=@n12.Result) {constB=10}) {constB=0.5}) {constB=0}
  Constant3Vector()
  n13 = TextureSample(UVs=@n11) {samplerType=SAMPLERTYPE_LinearColor, texture=/Game/Global/Textures/Default/Caustics/03_WaterFoam_AccentSwirls_H}
  n14 = ScalarParameter() {defaultValue=4, parameterName="Contrast (S)"}
  n15 = TextureSampleParameter2D(UVs=@n11) {parameterName=Normal, samplerType=SAMPLERTYPE_Normal, texture=/Game/Packages/Fortress_Sky/Materials/2Normal}
  LinearInterpolate(
   A=VectorParameter() {parameterName=Color_dark, defaultValue=<rgba 0.0156 0.0156 0.0156 1>}.RGB
   B=LinearInterpolate(
    A=VectorParameter() {parameterName=Color_Middle, defaultValue=<rgba 0.0833 0.0833 0.0833 1>}.RGB
    B=VectorParameter() {parameterName=Color_Bright, defaultValue=<rgba 1.9531 2.1092 3 1>}.RGB
    Alpha=Power(Base=@n13.R, Exp=@n14)
   )
   Alpha=Power(Base=@n15.R, Exp=ScalarParameter() {defaultValue=2, parameterName=Dark_Exp})
  )
  n16 = TextureCoordinate() {vTiling=2}
  n17 = Power(Base=DotProduct(
    A=SkyAtmosphereLightDirection()
    B=TextureSample(UVs=@n16) {samplerType=SAMPLERTYPE_Normal, texture=/Game/Global/Textures/Default/Noise/T_SwirlNoise_N}.RGB
   )) {constExponent=10}
  n18 = Power(Base=LinearInterpolate(Alpha=Saturate(None=@n17)) {constA=-10, constB=2}) {constExponent=10}
  n19 = ScalarParameter() {defaultValue=1, parameterName=Emissive_Dark}
  n20 = ScalarParameter() {defaultValue=1, parameterName=Emissive_Light}
  n21 = Power(Base=DotProduct(
    A=SkyAtmosphereLightDirection()
    B=TextureSample() {samplerType=SAMPLERTYPE_Normal, texture=&T_SwirlNoise_N}.RGB
   ))
  n22 = Multiply(
   A=TextureCoordinate() {vTiling=0.1}
   B=ComponentMask(None=VectorParameter() {parameterName=U_andV, defaultValue=<rgba 1 1 0 0>}.RGB) {g=true, r=true}
  )
  n23 = ComponentMask(None=VectorParameter() {parameterName=U_andV_offset, defaultValue=<rgba 1 1 0 0>}.RGB) {g=true, r=true}
  n24 = Desaturation(
   None=TextureSampleParameter2D(UVs=Add(A=@n22, B=@n23)) {parameterName=Emissive_Tex, texture=/ClashGameplay/Athena/Effects/Particles/Skydiving/NS_Contrail_Clash/T_Clash_Tendril_Bottom}.RGB
   Fraction=ScalarParameter() {parameterName=Des}
  )
  n25 = VectorParameter() {parameterName=Emissive_Color, defaultValue=<rgba 1 1 1 0>}
  n26 = Add(
   A=Multiply(A=Saturate(None=@n18)) {constB=50}
   B=Multiply(A=LinearInterpolate(A=@n19, B=@n20, Alpha=@n21) {constA=1}, B=Multiply(A=@n24, B=@n25.RGB)) {constA=1}
  )
  n27 = ScalarParameter() {parameterName=Add_Emissive}
  n28 = VectorParameter() {group=Color_time_of_day, parameterName=Color_Morning, defaultValue=<rgba 1 1 1 1>}
  n29 = VectorParameter() {group=Color_time_of_day, parameterName=Color_Day, defaultValue=<rgba 1 1 1 1>}
  n30 = VectorParameter() {group=Color_time_of_day, parameterName=Color_Evening, defaultValue=<rgba 1 1 1 1>}
  n31 = VectorParameter() {group=Color_time_of_day, parameterName=Color_Night, defaultValue=<rgba 1 1 1 1>}
  MaterialFunctionCall(Emissive=Multiply(
    A=LinearInterpolate(B=@n26, Alpha=@n27)
    B=MaterialFunctionCall(Color_Morning=@n28.RGB, Color_Day=@n29.RGB, Color_Evening=@n30.RGB, Color_Night=@n31.RGB) {materialFunction=&Color_Time_of_Day}.Result
   )) {materialFunction=&MF_FX_EmissiveAdjustments}
  n32 = CurveAtlasRowParameter(CurveTime=ComponentMask(None=TextureCoordinate()) {g=true}) {
   atlas=/Game/Athena/LinearColorCurves/TempAthenaCurveAtlas
   curve=/Game/Athena/Prototype/Blueprints/NightNight/FX/FinalImpact/Curve_HT_Tomato_RC
   defaultValue=11
  }
  Multiply(A=@n32.G)
  Multiply(B=ScalarParameter() {defaultValue=0.6, parameterName=ConeSizeScaleBaseOnLength})
  ScalarParameter() {parameterName=Roughness}
  ScalarParameter() {defaultValue=0.2, parameterName=Metallic}
  ScalarParameter() {parameterName=Specular}
  n33 = TextureCoordinate()
  n34 = Subtract(A=VectorParameter() {parameterName=TargetPosition, defaultValue=<rgba 0 0 0 1>}.RGB, B=ObjectPositionWS())
  n35 = Multiply(A=Normalize(VectorInput=@n34), B=ScalarParameter() {parameterName=ReduceLength}) {constB=-50}
  n36 = Normalize(VectorInput=VertexNormalWS())
  n37 = MaterialFunctionCall(Vector 3=@n34) {materialFunction=/Engine/Functions/Engine_MaterialFunctions02/Utility/VectorLength}
  n38 = ScalarParameter() {defaultValue=0.6, parameterName=ConeSizeScaleBaseOnLength}
  n39 = Clamp(None=Multiply(A=Multiply(A=@n37.V3 Length, B=@n38)) {constB=0.002}) {maxDefault=0.5, minDefault=0.2}
  n40 = ScalarParameter() {defaultValue=60, parameterName=ConeIntensity}
  n41 = Add(
   A=Multiply(A=ComponentMask(None=@n33) {g=true}, B=Add(A=@n35, B=@n34))
   B=Multiply(A=Multiply(A=@n36, B=@n32.G) {constB=0}, B=Multiply(A=@n39, B=@n40))
  )
  n42 = Multiply(A=Multiply(A=@n15.R, B=VertexNormalWS()), B=ScalarParameter() {parameterName=DisplaceAmount}) {constB=10}
  n43 = CurveAtlasRowParameter(CurveTime=ComponentMask(None=TextureCoordinate()) {g=true}) {
   atlas=&TempAthenaCurveAtlas
   curve=&Curve_HT_Tomato_RC
   defaultValue=11
  }
  n44 = TextureCoordinate()
  n45 = MaterialFunctionCall(Vector 3=@n34) {materialFunction=&VectorLength}
  n46 = Multiply(A=ScalarParameter() {defaultValue=1, parameterName=SinLength}) {constB=3000}
  n47 = Add(
   A=Multiply(A=ScalarParameter() {parameterName=SinMotion}, B=ScalarParameter() {defaultValue=1, parameterName=SinSpeed})
   B=Multiply(A=ComponentMask(None=@n44) {g=true}, B=Divide(A=@n45.V3 Length, B=@n46) {constB=3000})
  )
  n48 = Sine(None=@n47)
  n49 = Normalize(VectorInput=@n34)
  n50 = Tangent!(None=@n49)
  n51 = Normalize(VectorInput=@n50)
  n52 = ScalarParameter() {defaultValue=50, parameterName=SinIntensity}
  n53 = Cosine(None=Add(A=@n47, B=ScalarParameter() {parameterName=CosineOffset}))
  n54 = Normalize(VectorInput=CrossProduct(A=@n49, B=@n50))
  n55 = Add(A=Multiply(A=Multiply(A=@n48, B=@n51), B=@n52), B=Multiply(A=@n52, B=Multiply(A=@n53, B=@n54)))
  StaticSwitchParameter(True=Add(A=Add(A=@n41, B=@n42), B=Multiply(A=@n43.B, B=@n55)), False=Constant()) {defaultValue=true, parameterName=Use_WPO}
 ]
}
